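# Masked, two-sided lit material for the beach variant of pirate chest dirt, previewed on the booty dirt mesh.
# Desaturated dirt texture blends two tint colors, then vertex green blends in cliff texture; a flattened dirt normal adds detail.
# Inverted vertex red, raised to a power and offset by noise, forms a hard-edged cutout mask.
# Noise-fed temporal dither, apparently for pixel depth offset, runs only at High quality off mobile.
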Material /Game/Effects/Fort_Effects/Gameplay/PirateChests/M_Chest_Beach
graph {
 properties {
  blendMode=BLEND_Masked
  pixelDepthOffsetMode=PDOM_Legacy
  previewMesh=&/sm_booty_dirt
  twoSided=true
  usedShadingModels=Default Lit
 }
 nodes [
  Constant()
  n1 = TextureSample(UVs=TextureCoordinate() {uTiling=2, vTiling=2}) {texture=/Game/Packages/DS_Fortnite_Terrain_NoLOD/Textures/T_Desert_Darkest_Dirt_D}
  n2 = VertexColor()
  LinearInterpolate(
   A=LinearInterpolate(
    A=VectorParameter() {parameterName=Dirt Color 1, defaultValue=<rgba 0.4948 0.3208 0.1509 1>}.RGB
    B=VectorParameter() {parameterName=Dirt Color 2, defaultValue=<rgba 0.7049 0.448 0.2495 1>}.RGB
    Alpha=Desaturation(None=@n1.RGB)
   )
   B=TextureSample(UVs=TextureCoordinate() {uTiling=3, vTiling=3}) {texture=/Game/Packages/DS_Fortnite_Terrain_NoLOD/Textures/T_CliffTileB_D}.G
   Alpha=@n2.G
  )
  MaterialFunctionCall(
   Normal=TextureSample() {samplerType=SAMPLERTYPE_Normal, texture=/Game/Packages/DS_Fortnight/ExperimentalTextures/Dirt_N}.RGB
   Flatness=ScalarParameter() {defaultValue=0.5, parameterName=Normal Flatness}
  ) {materialFunction=&FlattenNormal}
  OneMinus(None=@n1.G)
  n3 = Power(Base=OneMinus(None=@n2.R))
  n4 = TextureSample(UVs=TextureCoordinate() {uTiling=3, vTiling=5}) {samplerSource=SSM_Wrap_WorldGroupSettings, texture=&T_NoisePack_01}
  n5 = Subtract(A=@n3, B=@n4.B)
  Saturate(None=Ceil(None=Add(A=@n3, B=@n5)))
  n6 = Constant()
  n7 = QualitySwitch(
   Default=@n6
   Low=@n6
   High=Multiply(
    A=MaterialFunctionCall(Offset=@n4.G, Offset2=@n4.G) {materialFunction=&DitherTemporalAA_Fortnite}.Result
    B=ScalarParameter() {defaultValue=3, parameterName=Dither Strength}
   )
   Medium=@n6
  )
  ShadingPathSwitch(Default=@n7, Deferred=@n7, Mobile=@n6) {desc=Migrated from FeatureLevelSwitch}
 ]
}
