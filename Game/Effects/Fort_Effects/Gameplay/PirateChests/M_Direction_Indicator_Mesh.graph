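# Default Lit material for a pirate chest direction indicator mesh, usable on Cascade and Niagara mesh particles.
# A Fresnel term with an adjustable exponent is raised to a second power parameter to shape the rim falloff.
# That falloff blends in an overbright orange Fresnel Color, which feeds the MF_FX_EmissiveAdjustments function as emissive.
# A red Color parameter and two constants are also present, with no connections shown between them and the glow chain.
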Material /Game/Effects/Fort_Effects/Gameplay/PirateChests/M_Direction_Indicator_Mesh
graph {
 properties {bUsedWithMeshParticles=true, bUsedWithNiagaraMeshParticles=true, pixelDepthOffsetMode=PDOM_Legacy, usedShadingModels=Default Lit}
 nodes [
  Constant()
  Constant() {r=1}
  n1 = Fresnel(ExponentIn=ScalarParameter() {defaultValue=8.7807, parameterName=Fresnel Exponent})
  n2 = ScalarParameter() {defaultValue=2, parameterName=Fresnel Power}
  MaterialFunctionCall(Emissive=LinearInterpolate(B=VectorParameter() {parameterName=Fresnel Color, defaultValue=<rgba 12 3.5422 0.6875 1>}.RGB, Alpha=Power(Base=@n1, Exp=@n2)) {constB=0}) {materialFunction=&MF_FX_EmissiveAdjustments}
  VectorParameter() {parameterName=Color, defaultValue=<rgba 1 0 0 1>}
 ]
}
