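# Masked, unlit, two-sided particle material for pirate chest effects that renders drifting scraps of an old treasure map.
# Color comes from a tiled old-map texture multiplied by particle color and an ember SubUV channel, optionally tinted by time of day, then passed through emissive adjustments.
# A world-position-offset style chain builds sine-based horizontal drift and age-driven height, varied per particle by random values and dynamic parameters.
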Material /Game/Effects/Fort_Effects/Gameplay/PirateChests/M_Map_Bits
graph {
 properties {
  bUsedWithNiagaraSprites=true
  bUsedWithParticleSprites=true
  blendMode=BLEND_Masked
  pixelDepthOffsetMode=PDOM_Legacy
  shadingModel=MSM_Unlit
  twoSided=true
  usedShadingModels=Unlit
 }
 nodes [
  n1 = ScalarParameter() {defaultValue=0.12, parameterName=WPOSpeedMin}
  n2 = ScalarParameter() {defaultValue=0.25, parameterName=WPOSpeedMax}
  n3 = ParticleRandom()
  n4 = DynamicParameter() {paramNames [Random,Random2,Param3,Param4]}
  n5 = Add(A=@n3, B=@n4.Random)
  n6 = Multiply(A=@n3) {constB=2.3}
  n7 = Add(A=Multiply(A=Time(), B=LinearInterpolate(A=@n1, B=@n2, Alpha=@n5)), B=Add(A=Frac(None=@n6), B=@n4.Random2))
  n8 = Sine(None=@n7) {period=5.14}
  n9 = Sine(None=@n7) {period=4.12}
  n10 = ScalarParameter() {defaultValue=12, parameterName=WPOHorizontalDrift}
  n11 = ParticleRelativeTime()
  n12 = Add(A=@n11) {constB=-0.3}
  n13 = LinearInterpolate(
   A=Multiply(A=@n4.Random2, B=ScalarParameter() {defaultValue=0.25, parameterName=MinWPOHorizontalDrift})
   Alpha=ConstantBiasScale(None=Sine(None=@n12) {period=1.5})
  ) {constA=1}
  n14 = ScalarParameter() {defaultValue=2, parameterName=WPOHeightLerpPower}
  AppendVector(
   A=Multiply(A=AppendVector(A=@n8, B=@n9), B=Multiply(A=@n10, B=@n13)) {constA=1}
   B=LinearInterpolate(B=ScalarParameter() {defaultValue=18, parameterName=WPOHeightLimit}, Alpha=Power(Base=@n11, Exp=@n14) {constExponent=2.4}) {constAlpha=0}
  )
  n15 = ParticleColor()
  n16 = TextureCoordinate() {vTiling=0.5}
  n17 = Constant2Vector() {g=0.85, r=6}
  n18 = Frac(None=Add(A=Multiply(A=@n16, B=@n17) {constB=6}, B=Constant2Vector() {g=0.065}))
  n19 = MaterialFunctionCall(Float3=ParticleSubUV!(UVs, Tex, Apply View MipBias) {texture=/Game/Effects/Fort_Effects/Textures/Motes/T_Ash_Embers_Glow_SubUV}.RGB) {materialFunction=&BreakOutFloat3Components}
  n20 = Multiply(
   A=@n15.RGB
   B=Multiply(
    A=TextureSample(UVs=@n18) {texture=/Game/Missions/Quests/ReactiveQuests/CannyValley/Fetch/Props/Textures/T_Old_Map_D}.RGB
    B=Multiply(A=@n15.RGB, B=@n19.G)
   )
  )
  n21 = VectorParameter() {group=Smoke Colors, parameterName=2 - Morning Mids Smoke, sortPriority=1, defaultValue=<rgba 0.445 0.2656 0.155 1>}
  n22 = VectorParameter() {group=Smoke Colors, parameterName=2 - Daytime Mids Smoke, sortPriority=2, defaultValue=<rgba 0.6484 0.7086 0.75 1>}
  n23 = VectorParameter() {group=Smoke Colors, parameterName=2 - Evening Mids Smoke, sortPriority=3, defaultValue=<rgba 0.79 0.2832 0.1857 1>}
  n24 = VectorParameter() {group=Smoke Colors, parameterName=2 - Nighttime Mids Smoke, sortPriority=4, defaultValue=<rgba 0.108 0.2106 0.27 1>}
  n25 = StaticSwitchParameter(
   True=@n20
   False=Multiply(
    A=MaterialFunctionCall(Color_Morning=@n21.RGB, Color_Day=@n22.RGB, Color_Evening=@n23.RGB, Color_Night=@n24.RGB) {desc=Mids, materialFunction=&Color_Time_of_Day}.Result
    B=@n20
   )
  ) {
   desc="If you want the same colors to be used throughout any time of day, activate this bool. Uses daytime value."
   group=Smoke Colors
   parameterName=Ignore ToD Breakout
  }
  MaterialFunctionCall(Emissive=@n25) {materialFunction=&MF_FX_EmissiveAdjustments}
 ]
}
