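# Masked, lit material for mesh particles that tears and dissolves an old treasure map, with map color and normal textures, for pirate chest effects.
# A 3D box mask around the object position, with falloff driven by the MapTearingParam dynamic parameter, sets the dissolve region.
# Tileable noise, offset by Param2, breaks up that mask, and a Ceil turns the inverted product into a hard-edged cutout.
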Material /Game/Effects/Fort_Effects/Gameplay/PirateChests/M_Map_Dissolve
graph {
 properties {
  bUsedWithMeshParticles=true
  bUsedWithNiagaraMeshParticles=true
  blendMode=BLEND_Masked
  pixelDepthOffsetMode=PDOM_Legacy
  usedShadingModels=Default Lit
 }
 nodes [
  n1 = DynamicParameter() {paramNames [MapTearingParam, Param2, Param3, Param4]}
  n2 = Saturate(None=MaterialFunctionCall!(
    A
    B=ObjectPositionWS()
    Bounds=VectorParameter() {parameterName=Object Scale Bounds, defaultValue=<rgba 0 500 500 1>}.RGB
    Edge Falloff=@n1.MapTearingParam
   ) {materialFunction=/Engine/Functions/Engine_MaterialFunctions01/Gradient/BoxMask-3D}.Result)
  n3 = ScalarParameter() {defaultValue=1, parameterName=TearU Tile}
  n4 = ScalarParameter() {defaultValue=1, parameterName=TearV Tile}
  n5 = Add(A=Multiply(A=TextureCoordinate() {uTiling=0.2, vTiling=1.4}, B=AppendVector(A=@n3, B=@n4)), B=@n1.Param2)
  n6 = Add(A=Multiply(A=TextureSample(UVs=@n5) {texture=/Game/Global/Textures/Default/Noise/TankMoteNoise}.G, B=@n2), B=@n2)
  Ceil(None=Multiply(A=OneMinus(None=@n2), B=OneMinus(None=@n6)))
  Constant()
  Constant() {r=1}
  n7 = Time()
  Multiply(A=ConstantBiasScale(None=Sine(None=@n7)), B=Constant() {r=80})
  ParticlePositionWS()
  TextureSample() {texture=/Game/Missions/Quests/ReactiveQuests/CannyValley/Fetch/Props/Textures/T_Old_Map_D}
  TextureSample() {samplerType=SAMPLERTYPE_Normal, texture=/Game/Athena/Items/ForagedItems/Booty/PROTO_Athena_T_Old_Map_N}
  WorldPosition()
 ]
}
