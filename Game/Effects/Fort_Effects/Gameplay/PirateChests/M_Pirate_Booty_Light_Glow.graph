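# Unlit, two-sided translucent mesh-particle material for the light glow on pirate booty chests.
# Three panning noise textures, masked by second-UV gradients, blend two HDR colors tinted by particle color and form opacity with particle alpha and double-sided Fresnel.
# Scrolling cloud noise drives a masked world-position-offset term.
# Mobile and low quality skip depth fade and zero that offset term.
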Material /Game/Effects/Fort_Effects/Gameplay/PirateChests/M_Pirate_Booty_Light_Glow
graph {
 properties {
  bUsedWithMeshParticles=true
  bUsedWithNiagaraMeshParticles=true
  blendMode=BLEND_Translucent
  pixelDepthOffsetMode=PDOM_Legacy
  shadingModel=MSM_Unlit
  twoSided=true
  usedShadingModels=Unlit
 }
 nodes [
  n1 = VectorParameter() {parameterName=Color A, defaultValue=<rgba 3 0.8856 0.1719 1>}
  n2 = VectorParameter() {parameterName=Color B, defaultValue=<rgba 8 2.4549 0.5833 1>}
  n3 = TextureCoordinate()
  n4 = TextureSample(UVs=Multiply(A=Panner(Coordinate=@n3) {speedX=0.01, speedY=0.0125}, B=Constant2Vector() {g=0.001, r=0.25})) {texture=/Game/Global/Textures/Default/Masks_and_Gradients/T_BlasterBeamDistort}
  n5 = TextureSample(UVs=Multiply(A=Panner(Coordinate=@n3) {speedX=-0.032, speedY=0.18}, B=Constant2Vector() {g=3, r=1})) {texture=&T_BlasterBeamDistort}
  n6 = Multiply(A=Panner(Coordinate=TextureCoordinate()) {speedY=0.225}, B=Constant2Vector() {g=0.25, r=1})
  n7 = MaterialFunctionCall(Float2=TextureCoordinate() {coordinateIndex=1}) {materialFunction=&BreakOutFloat2Components}
  n8 = OneMinus(None=@n7.R)
  n9 = Multiply(A=Multiply(A=@n8, B=@n8), B=Saturate(None=Multiply(A=@n7.R) {constB=1.4}))
  n10 = Multiply(
   A=Add(A=Add(A=@n4.G, B=@n5.G), B=TextureSample(UVs=@n6) {texture=&T_ShieldCaustics}.G)
   B=Multiply(A=@n9, B=@n7.G)
  )
  n11 = ParticleColor()
  MaterialFunctionCall(Emissive=Multiply(A=LinearInterpolate(A=@n1.RGB, B=@n2.RGB, Alpha=@n10), B=@n11.RGB)) {materialFunction=&MF_FX_EmissiveAdjustments}
  Multiply(A=@n4.G, B=@n4.G)
  Multiply(A=@n5.G, B=@n5.G)
  n12 = Multiply(
   A=Saturate(None=Multiply(A=@n11.A, B=@n10))
   B=MaterialFunctionCall(
    Brightness=ScalarParameter() {defaultValue=1, parameterName=Fresnel Brightness}
    Exponent=ScalarParameter() {defaultValue=2, parameterName=Fresnel Exponent}
   ) {materialFunction=&MF_Fresnel_FX}.Fresnel Double Sided
  )
  n13 = ScalarParameter() {defaultValue=12, parameterName=Depth_Fade_Distance}
  ShadingPathSwitch(Default=QualitySwitch(Default=DepthFade(Opacity=@n12, FadeDistance=@n13), Low=@n12), Mobile=@n12) {desc=Migrated from FeatureLevelSwitch}
  n14 = Panner(Coordinate=Multiply(A=TextureCoordinate() {coordinateIndex=1}, B=Constant() {r=0.2})) {speedY=0.1}
  n15 = Multiply(
   A=OneMinus(None=@n7.G)
   B=Multiply(
    A=TextureSample(UVs=@n14) {samplerType=SAMPLERTYPE_LinearGrayscale, texture=/Game/Athena/Prototype/Terrain/T_CloudCoverage}.RGB
    B=ScalarParameter() {defaultValue=20, parameterName=WPO_Intensity}
   ) {constB=18}
  )
  n16 = Constant()
  ShadingPathSwitch(Default=QualitySwitch(Default=Multiply(A=@n9, B=@n15), Low=@n16), Mobile=@n16) {desc=Migrated from FeatureLevelSwitch}
  VertexNormalWS()
 ]
}
