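# Unlit translucent deferred decal material for the pirate chest "X marks the spot" ground marker, with a fiery burn-in edge.
# Two counter-rotating smoke samples distort radial gradient UVs, and X Decal Fire Lerp drives the gradients' radius and hardness.
# Two offset-radius gradients form the hot edge band tinted between Color A and B, while the outer one masks the X texture's green channel.
# Mobile switches to a no-mip X texture.
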Material /Game/Effects/Fort_Effects/Gameplay/PirateChests/M_X_Ground_Decal
graph {
 properties {
  blendMode=BLEND_Translucent
  materialDomain=MD_DeferredDecal
  pixelDepthOffsetMode=PDOM_Legacy
  shadingModel=MSM_Unlit
  usedShadingModels=Unlit
 }
 nodes [
  Constant()
  Constant()
  Constant() {r=1}
  MaterialFunctionCall() {materialFunction=&DebugTimeSine}
  n1 = TextureCoordinate()
  n2 = Multiply(A=Rotator(Coordinate=@n1) {speed=0.333}) {constB=3}
  n3 = Multiply(A=Rotator(Coordinate=@n1) {speed=-0.1189}) {constB=3}
  n4 = ComponentMask(None=Multiply(
    A=TextureSample(UVs=@n2) {samplerSource=SSM_Wrap_WorldGroupSettings, texture=&T_Inky_Smoke_Tile}.RGB
    B=TextureSample(UVs=@n3) {samplerSource=SSM_Wrap_WorldGroupSettings, texture=&T_Inky_Smoke_Tile}.RGB
   )) {g=true, r=true}
  n5 = ScalarParameter() {defaultValue=0.45, parameterName=Flame Edge Distortion Multiplier}
  n6 = TextureCoordinate()
  n7 = ComponentMask(None=VectorParameter() {parameterName=Radial Gradient Offset, defaultValue=<rgba 0 0.045 0 1>}.RGB) {g=true, r=true}
  n8 = ScalarParameter() {defaultValue=0.125, parameterName=Distortion Offset Multiplier}
  n9 = Add(A=Add(A=Multiply(A=@n4, B=@n5) {constB=0.025}, B=Add(A=@n6, B=@n7)), B=Multiply(A=Multiply(A=@n5, B=@n8) {constB=0.125}) {constB=-1})
  n10 = ScalarParameter() {defaultValue=1.2, parameterName=BurnOffRadius}
  n11 = Add(A=@n10) {constB=0.25}
  n12 = ScalarParameter() {defaultValue=0.1, parameterName=X Decal Fire Lerp}
  n13 = LinearInterpolate(B=ScalarParameter() {defaultValue=150, parameterName=BurnOffHardness}, Alpha=@n12) {constA=0.5}
  n14 = Power(Base=MaterialFunctionCall(UVs=@n9, Radius=LinearInterpolate(B=@n11, Alpha=@n12), Density=@n13) {materialFunction=&RadialGradientExponential}.RadialGradientExponential) {constExponent=0.2}
  n15 = Constant() {r=0.15}
  n16 = Constant() {r=150}
  Multiply(
   A=@n14
   B=OneMinus(None=MaterialFunctionCall(UVs=@n9, Radius=@n15, Density=@n16) {materialFunction=&RadialGradientExponential}.RadialGradientExponential)
  )
  n17 = MaterialFunctionCall(UVs=@n9, Radius=LinearInterpolate(B=@n10, Alpha=@n12), Density=@n13) {materialFunction=&RadialGradientExponential}
  n18 = Multiply(A=OneMinus(None=Power(Base=@n17.RadialGradientExponential) {constExponent=0.2}), B=@n14)
  n19 = TextureSample() {samplerSource=SSM_Wrap_WorldGroupSettings, texture=&/t_x_marks_the_spot}
  n20 = TextureSample() {texture=&/t_x_marks_the_spot_NO_MIP}
  n21 = ShadingPathSwitch(Default=@n19.R, Mobile=@n20.R) {desc=Migrated from FeatureLevelSwitch}
  Multiply(
   A=LinearInterpolate(
    A=VectorParameter() {parameterName=Color B, defaultValue=<rgba 0 0 0 1>}.RGB
    B=VectorParameter() {parameterName=Color A, defaultValue=<rgba 5 1.1195 0 1>}.RGB
    Alpha=Saturate(None=@n18)
   )
   B=@n21
  )
  Multiply(A=@n21, B=VectorParameter() {parameterName=BaseColor, defaultValue=<rgba 0.8521 0.8521 0.8521 0>}.RGB)
  Saturate(None=Multiply(A=ShadingPathSwitch(Default=@n19.G, Mobile=@n20.G) {desc=Migrated from FeatureLevelSwitch}, B=@n14))
 ]
}
