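# Translucent unlit material for a pirate chest X ground glow, used on meshes and Niagara mesh particles.
# Panned noise with per-channel UV offsets and plasma drives chromatic HDR emissive and, with a ramp, particle alpha and optional fresnel, opacity; mobile uses flat color.
# Opacity rises with camera distance beyond an offset, with depth fade above low quality.
# Vertex-color-masked offset math scales the mesh and collapses it to its pivot inside that offset.
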Material /Game/Effects/Fort_Effects/Gameplay/PirateChests/M_X_Ground_Glow
graph {
 properties {
  bUsedWithInstancedStaticMeshes=true
  bUsedWithMeshParticles=true
  bUsedWithNanite=true
  bUsedWithNiagaraMeshParticles=true
  blendMode=BLEND_Translucent
  pixelDepthOffsetMode=PDOM_Legacy
  previewMesh=/JunoKlomboGP/Items/TreasureMap/FX/sm_x_glow_geo
  shadingModel=MSM_Unlit
  usedShadingModels=Unlit
 }
 nodes [
  Constant() {r=1}
  Fresnel(ExponentIn=ScalarParameter() {defaultValue=5, parameterName=FresnelExpo})
  n1 = Constant()
  n2 = Subtract(A=WorldPosition().XYZ, B=ObjectPositionWS())
  n3 = Multiply(
   A=ScalarParameter() {parameterName=Scale_Z}
   B=MaterialFunctionCall(X=@n1, Y=@n1, Z=ComponentMask(None=@n2) {b=true}) {materialFunction=&MakeFloat3}.Result
  ) {constB=-1}
  n4 = VertexColor()
  n5 = Subtract(A=ObjectPositionWS(), B=WorldPosition().XYZ)
  n6 = Multiply(A=ScalarParameter() {defaultValue=-1, parameterName=Scale_XY}, B=AppendVector(A=ComponentMask(None=@n5) {g=true, r=true}, B=Constant())) {constB=-1}
  n7 = VertexColor()
  n8 = Distance(A=ObjectPositionWS(), B=CameraPositionWS())
  n9 = ScalarParameter() {defaultValue=3000, parameterName=FadeDistanceOffset}
  n10 = ScalarParameter() {defaultValue=12000, parameterName=FadeDistance}
  n11 = Ceil(None=Divide(A=Subtract(A=@n8, B=@n9), B=@n10))
  LinearInterpolate(A=Add(A=Multiply(A=@n3, B=@n4.R), B=Multiply(A=@n6, B=@n7.R)), B=Multiply(A=@n2) {constB=-1}, Alpha=OneMinus(None=Clamp(None=@n11)))
  n12 = Multiply(A=TextureCoordinate() {uTiling=6}, B=ScalarParameter() {defaultValue=1, parameterName=Tile_Tex_1})
  n13 = Multiply(A=Time(), B=ScalarParameter() {defaultValue=1, parameterName=Movement Speed Multiplier})
  n14 = Multiply(A=TextureCoordinate() {vTiling=0.01}, B=ScalarParameter() {defaultValue=1, parameterName=Tile_Tex_0})
  n15 = AppendVector(A=ScalarParameter() {parameterName=U_Offset}, B=ScalarParameter() {parameterName=V_Offset})
  n16 = TextureSample(UVs=Panner(Coordinate=Add(A=@n14, B=@n15), Time=@n13) {speedX=0.012, speedY=-0.125}) {samplerSource=SSM_Wrap_WorldGroupSettings, texture=/Game/Global/Textures/Default/Noise/T_TileNoise_02}
  n17 = TextureSample(UVs=Panner(Coordinate=@n14, Time=@n13) {speedX=0.012, speedY=-0.125}) {samplerSource=SSM_Wrap_WorldGroupSettings, texture=&T_TileNoise_02}
  n18 = Panner(Coordinate=Add(A=@n14, B=Multiply(A=@n15) {constB=-1}), Time=@n13) {speedX=0.012, speedY=-0.125}
  n19 = Multiply(A=TextureCoordinate() {uTiling=4, vTiling=0.025}, B=ScalarParameter() {defaultValue=1, parameterName=Tile_Tex_2})
  n20 = AppendVector(A=ScalarParameter() {parameterName=U_Offset}, B=ScalarParameter() {parameterName=V_Offset})
  n21 = TextureSample(UVs=Panner(Coordinate=Add(A=@n19, B=@n20), Time=@n13) {speedX=0.025, speedY=-0.1}) {samplerSource=SSM_Wrap_WorldGroupSettings, texture=&T_TileNoise_02}
  n22 = TextureSample(UVs=Panner(Coordinate=@n19, Time=@n13) {speedX=0.025, speedY=-0.1}) {samplerSource=SSM_Wrap_WorldGroupSettings, texture=&T_TileNoise_02}
  n23 = Panner(Coordinate=Add(A=@n19, B=Multiply(A=@n20) {constB=-1}), Time=@n13) {speedX=0.025, speedY=-0.1}
  n24 = Add(
   A=AppendVector(
    A=AppendVector(A=@n16.G, B=@n17.G)
    B=TextureSample(UVs=@n18) {samplerSource=SSM_Wrap_WorldGroupSettings, texture=&T_TileNoise_02}.G
   )
   B=AppendVector(
    A=AppendVector(A=@n21.G, B=@n22.G)
    B=TextureSample(UVs=@n23) {samplerSource=SSM_Wrap_WorldGroupSettings, texture=&T_TileNoise_02}.G
   )
  )
  n25 = Add(
   A=TextureSample(UVs=Panner(Coordinate=@n12, Time=@n13) {speedX=0.125}) {samplerSource=SSM_Wrap_WorldGroupSettings, texture=/Game/Effects/Fort_Effects/Textures/Tile/T_Plasma_BallEdge}.G
   B=@n24
  )
  n26 = ParticleColor()
  n27 = Multiply(A=@n25, B=@n26.RGB)
  n28 = Multiply(A=VectorParameter() {parameterName=MobileEmissiveColor, defaultValue=<rgba 0.974 0.895 0.6583 0>}.RGB, B=@n26.RGB)
  n29 = ShadingPathSwitch(
   Default=MaterialFunctionCall(LDR Input=Saturate(None=@n27), HDR Tint=@n27) {materialFunction=&DeriveHDRfromLDR}.RemappedRange
   Mobile=@n28
  ) {desc=Migrated from FeatureLevelSwitch}
  n30 = Multiply(A=@n26.RGB, B=VectorParameter() {parameterName=EmissiveColor, defaultValue=<rgba 1 1 1 1>}.RGB)
  MaterialFunctionCall(Emissive=StaticSwitchParameter(
    True=ShadingPathSwitch(Default=@n29, Forward=@n28)
    False=ShadingPathSwitch(Default=@n30, Mobile=@n28) {desc=Migrated from FeatureLevelSwitch}
   ) {parameterName=HDR_Chromatic}) {materialFunction=&MF_FX_EmissiveAdjustments}
  n31 = DynamicParameter() {defaultValue=<rgba 0 1 1 1>, paramNames [ScaleZ,ScaleXY,FadeDistance,Param4]}
  n32 = LinearInterpolate(A=Desaturation(None=@n24), B=Desaturation(None=@n25), Alpha=@n31.ScaleZ) {constAlpha=1}
  n33 = Constant()
  n34 = Add(A=Constant() {r=1}, B=Multiply(A=@n31.ScaleXY, B=ScalarParameter() {parameterName=ScaleXY}))
  n35 = TextureSample() {texture=/Game/Global/Textures/Default/Masks_and_Gradients/T_Soft_Ramp_Vert_03}
  n36 = Multiply(
   A=@n26.A
   B=Multiply(A=ShadingPathSwitch(Default=@n32) {desc=Migrated from FeatureLevelSwitch}, B=LinearInterpolate(A=@n33, B=@n34, Alpha=@n35.R))
  )
  n37 = DotProduct(A=CameraVectorWS(), B=VertexNormalWS())
  n38 = Power(Base=Saturate(None=Abs(None=@n37)), Exp=ScalarParameter() {defaultValue=3, parameterName=Fresnel Power})
  n39 = Multiply(A=OneMinus(None=VertexColor().G), B=StaticSwitchParameter(True=Multiply(A=@n36, B=@n38), False=@n36) {parameterName=UseFresnelEdge})
  n40 = ScalarParameter() {defaultValue=1, parameterName=DepthFadeScalar}
  n41 = QualitySwitch(Default=DepthFade(Opacity=@n39, FadeDistance=Multiply(A=@n31.FadeDistance, B=@n40)), Low=@n39)
  n42 = ActorPositionWS()
  n43 = CameraPositionWS()
  n44 = Divide(A=Subtract(A=Distance(A=@n42, B=@n43), B=@n9), B=@n10)
  Multiply(
   A=Multiply(A=ShadingPathSwitch(Default=@n41, Mobile=@n39) {desc=Migrated from FeatureLevelSwitch}, B=Clamp(None=@n44))
   B=ScalarParameter() {defaultValue=1, parameterName=Final Opacity}
  )
  ScalarParameter() {defaultValue=-1, parameterName=TEMP}
 ]
}
